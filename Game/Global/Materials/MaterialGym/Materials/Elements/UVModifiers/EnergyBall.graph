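# EnergyBall is a Default Lit material in the MaterialGym UVModifiers elements folder, named for an energy ball look.
# Its graph is a single call to the OverlayTransparent function MF_EnergyNoise_SilverBell_Generic, which exposes Material, Alpha and FrontEndOverride pins.
# The noise logic lives inside that function, and this file shows no parameters or UV math of its own.
# Automatic usage flag setting in the editor is disabled.
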Material /Game/Global/Materials/MaterialGym/Materials/Elements/UVModifiers/EnergyBall
graph {
 properties {bAutomaticallySetUsageInEditor=false, usedShadingModels=Default Lit}
 nodes [MaterialFunctionCall!(Material, Alpha, FrontEndOverride) {materialFunction=/Game/Global/Materials/MaterialGym/Materials/OverlayTransparent/MF_EnergyNoise_SilverBell_Generic}]
}
