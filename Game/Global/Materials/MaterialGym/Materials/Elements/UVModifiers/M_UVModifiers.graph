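# Unlit material-gym element that prototypes a spherical UV distortion, with no connection to material outputs shown.
# Centered UVs, divided by Size and multiplied by Curvature, yield a hemisphere height from a square root, feeding Normal, Height and ceiling-based Mask reroutes.
# Size-scaled coordinates are divided by height plus a constant, then remapped into a UV reroute that samples a water-foam swirl texture, giving a stereographic-style bulge.
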Material /Game/Global/Materials/MaterialGym/Materials/Elements/UVModifiers/M_UVModifiers
graph {
 properties {bAutomaticallySetUsageInEditor=false, shadingModel=MSM_Unlit, usedShadingModels=Unlit}
 nodes [
  Constant2Vector() {desc=size, g=1, r=1}
  Constant2Vector() {desc=curvatures, g=1, r=1}
  n1 = Divide(A=ConstantBiasScale(None=TextureCoordinate()) {bias=-0.5, scale=2}, B=ScalarParameter() {defaultValue=1, parameterName=Size})
  n2 = Multiply(A=@n1, B=ScalarParameter() {defaultValue=1, parameterName=Curvature})
  n3 = DotProduct(A=@n2, B=@n2)
  n4 = SquareRoot(None=OneMinus(None=Saturate(None=@n3)))
  NamedRerouteDeclaration(None=AppendVector(A=@n2, B=@n4)) {name=Normal, nodeColor=<rgba 1 0 0.2812 1>}
  NamedRerouteDeclaration(None=Ceil(None=@n4)) {name=Mask, nodeColor=<rgba 0 0.6016 1 1>}
  NamedRerouteDeclaration(None=@n4) {name=Height, nodeColor=<rgba 0.9219 1 0 1>}
  n5 = Divide(A=@n1, B=Add(A=@n4))
  TextureSample(UVs=NamedRerouteDeclaration(None=ConstantBiasScale(None=@n5)) {name=UV, nodeColor=<rgba 0.7578 0 1 1>}) {samplerType=SAMPLERTYPE_LinearColor, texture=/Game/Global/Textures/Default/Caustics/03_WaterFoam_AccentSwirls_H}
 ]
}
